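# Sound cue for the first-person fire sound of the Flintlock pistol.
# Each path mixes a random Close shot with a slightly quieter random PushBack layer delayed by a fixed 0.2 seconds, then applies random volume modulation.
# A quality-level switch gives the first tier four Close variations, while the other three tiers share a branch with three.
# It uses first-person weapon attenuation with overrides, a ranged-fire sound class and first-person gunfire concurrency.
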SoundCue /Game/Sounds/Fort_Weapons/pistol/Flintlock/Pistol_Flintlock_Fire_1P_Cue
root=QualityLevel_0
nodes [
 {id=Delay_0, children [Random_2], properties {delayMax=0.2, delayMin=0.2}}
 {id=Delay_1, children [Random_3], properties {delayMax=0.2, delayMin=0.2}}
 {id=Mixer_0, children [Random_1,Delay_1], properties {inputVolume [1,0.8]}}
 {id=Mixer_3, children [Random_0,Delay_0], properties {inputVolume [1,0.8]}}
 {id=Modulator_1, children [Mixer_0], properties {volumeMax=1.2, volumeMin=1.1}}
 {id=Modulator_6, children [Mixer_3], properties {volumeMax=1.2, volumeMin=1.1}}
 {id=QualityLevel_0, children [Modulator_6,Modulator_1,Modulator_1,Modulator_1]}
 {id=Random_0, children [WavePlayer_4,WavePlayer_5,WavePlayer_6,WavePlayer_7], properties {weights [1,1,1,1]}}
 {id=Random_1, children [WavePlayer_8, WavePlayer_9, WavePlayer_11], properties {weights [1,1,1]}}
 {id=Random_2, children [WavePlayer_3,WavePlayer_2], properties {weights [1,1]}}
 {id=Random_3, children [WavePlayer_12, WavePlayer_10], properties {weights [1,1]}}
 {id=WavePlayer_10, properties {soundWaveAssetPtr=&/Pistol_Flintlock_Fire_PushBack_03}}
 {id=WavePlayer_11, properties {soundWaveAssetPtr=&/Pistol_Flintlock_Fire_Close_04}}
 {id=WavePlayer_12, properties {soundWaveAssetPtr=&/Pistol_Flintlock_Fire_PushBack_04}}
 {id=WavePlayer_2, properties {soundWaveAssetPtr=&Pistol_Flintlock_Fire_PushBack_03}}
 {id=WavePlayer_3, properties {soundWaveAssetPtr=&Pistol_Flintlock_Fire_PushBack_04}}
 {id=WavePlayer_4, properties {soundWaveAssetPtr=&/Pistol_Flintlock_Fire_Close_01}}
 {id=WavePlayer_5, properties {soundWaveAssetPtr=&/Pistol_Flintlock_Fire_Close_02}}
 {id=WavePlayer_6, properties {soundWaveAssetPtr=&/Pistol_Flintlock_Fire_Close_03}}
 {id=WavePlayer_7, properties {soundWaveAssetPtr=&Pistol_Flintlock_Fire_Close_04}}
 {id=WavePlayer_8, properties {soundWaveAssetPtr=&Pistol_Flintlock_Fire_Close_01}}
 {id=WavePlayer_9, properties {soundWaveAssetPtr=&Pistol_Flintlock_Fire_Close_02}}
]
properties {
 attenuationSettings=&Fort_Weapon_Projectile_FirstPerson_Athena_Att
 maxDistance=2097152
 soundClassObject=&Fort_WeaponRanged_Fire_1P_PSM
 volumeMultiplier=1
 attenuationOverrides {falloffDistance=4100, attenuationShapeExtents {x=900}}
 concurrencySet [&Weapon_Gunfire_FirstPerson_Conc]
}
