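# Sound cue for the flintlock pistol's gunshot, layering close, medium and distant fire sounds.
# Each layer randomly picks a wave variant, varies it through a modulator, then applies distance-specific attenuation and a sound class override before a weighted mixer combines all three.
# The close layer uses a third-person fire class, while medium and distant share the distant fire class.
# Local and global gunfire concurrency sets apply to the cue.
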SoundCue /Game/Sounds/Fort_Weapons/pistol/Flintlock/Pistol_Flintlock_Fire_Cue
root=Mixer_2
nodes [
 {
  id=Attenuation_119
  children [Modulator_2]
  properties {attenuationSettings=&Fort_Weapon_Projectile_Medium_Mono_Athena_Att}
 }
 {
  id=Attenuation_2
  children [Modulator_5]
  properties {attenuationSettings=&Fort_Weapon_Projectile_Distant_Mono_Athena_Att}
 }
 {id=Attenuation_3, children [Modulator_6], properties {attenuationSettings=&Fort_Weapon_Projectile_Close_Athena_Att}}
 {id=Mixer_2, children [SoundClass_3, SoundClass_76, SoundClass_2], properties {inputVolume [1,0.8,0.5]}}
 {id=Modulator_2, children [Random_5], properties {volumeMax=1.6, volumeMin=1.4}}
 {id=Modulator_5, children [Random_7], properties {volumeMax=1.1, volumeMin=1}}
 {id=Modulator_6, children [Random_4], properties {pitchMin=1, volumeMax=1.3, volumeMin=1.2}}
 {
  id=Random_4
  children [WavePlayer_15, WavePlayer_16, WavePlayer_17, WavePlayer_18]
  properties {bShouldExcludeFromBranchCulling=true, weights [1,1,1,1]}
 }
 {id=Random_5, children [WavePlayer_19, WavePlayer_20, WavePlayer_21], properties {bShouldExcludeFromBranchCulling=true, weights [1,1,1]}}
 {id=Random_7, children [WavePlayer_22, WavePlayer_23, WavePlayer_24], properties {bShouldExcludeFromBranchCulling=true, weights [1,1,1]}}
 {
  id=SoundClass_2
  children [Attenuation_2]
  properties {soundClassOverride=&Fort_WeaponRanged_Fire_Distant}
 }
 {
  id=SoundClass_3
  children [Attenuation_3]
  properties {soundClassOverride=&Fort_WeaponRanged_Fire_3P_PSM}
 }
 {
  id=SoundClass_76
  children [Attenuation_119]
  properties {soundClassOverride=&Fort_WeaponRanged_Fire_Distant}
 }
 {id=WavePlayer_15, properties {soundWaveAssetPtr=&/Pistol_Flintlock_Fire_Close_01}}
 {id=WavePlayer_16, properties {soundWaveAssetPtr=&/Pistol_Flintlock_Fire_Close_02}}
 {id=WavePlayer_17, properties {soundWaveAssetPtr=&/Pistol_Flintlock_Fire_Close_03}}
 {id=WavePlayer_18, properties {soundWaveAssetPtr=&/Pistol_Flintlock_Fire_Close_04}}
 {id=WavePlayer_19, properties {soundWaveAssetPtr=&/Pistol_Flintlock_Fire_Medium_01}}
 {id=WavePlayer_20, properties {soundWaveAssetPtr=&/Pistol_Flintlock_Fire_Medium_02}}
 {id=WavePlayer_21, properties {soundWaveAssetPtr=&/Pistol_Flintlock_Fire_Medium_03}}
 {id=WavePlayer_22, properties {soundWaveAssetPtr=&/Pistol_Flintlock_Fire_Distant_02}}
 {id=WavePlayer_23, properties {soundWaveAssetPtr=&/Pistol_Flintlock_Fire_Distant_01}}
 {id=WavePlayer_24, properties {soundWaveAssetPtr=&/Pistol_Flintlock_Fire_Distant_03}}
]
properties {
 maxDistance=27500
 volumeMultiplier=0.85
 attenuationOverrides {falloffDistance=4100, attenuationShapeExtents {x=900}}
 concurrencySet [&Weapon_Gunfire_Local_Conc, &Weapon_Gunfire_Global_Conc]
}
