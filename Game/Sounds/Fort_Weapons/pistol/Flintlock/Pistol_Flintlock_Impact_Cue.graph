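# Sound cue for the flintlock pistol's bullet impact, reusing the Dragon shotgun's impact explosion waves.
# A random node picks one of three equally weighted explosion variants, then a modulator varies its pitch and volume on each play.
# It uses the bullet impact attenuation, the impacts sound class and the rifle bullet impact concurrency group.
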SoundCue /Game/Sounds/Fort_Weapons/pistol/Flintlock/Pistol_Flintlock_Impact_Cue
root=Modulator_6
nodes [
 {id=Modulator_6, children [Random_3], properties {pitchMax=1.2, pitchMin=0.8, volumeMax=1.1, volumeMin=0.9}}
 {id=Random_3, children [WavePlayer_29, WavePlayer_30, WavePlayer_31], properties {weights [1,1,1]}}
 {id=WavePlayer_29, properties {soundWaveAssetPtr=/Game/Sounds/Fort_Weapons/Dragon/Shotgun/Dragon_Shotgun_ImpactExplosion_01}}
 {id=WavePlayer_30, properties {soundWaveAssetPtr=/Game/Sounds/Fort_Weapons/Dragon/Shotgun/Dragon_Shotgun_ImpactExplosion_03}}
 {id=WavePlayer_31, properties {soundWaveAssetPtr=/Game/Sounds/Fort_Weapons/Dragon/Shotgun/Dragon_Shotgun_ImpactExplosion_02}}
]
properties {
 attenuationSettings=/Game/Sounds/Attenuation/Fort_Bullet_Impact_Att
 maxDistance=5300
 soundClassObject=&Fort_Impacts
 volumeMultiplier=1
 concurrencySet [&RifleBulletImpact_Conc]
}
