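# Sound cue for the end of the flintlock pistol reload.
# A single wave player plays the flintlock revolver cock sample at lowered volume and slightly lowered pitch, with no randomization or mixing.
# It uses the weapon foley sound class and the shared weapon reload attenuation.
# A concurrency override blocks new instances once the limit is reached instead of stopping playing ones.
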SoundCue /Game/Sounds/Fort_Weapons/pistol/Flintlock/Pistol_Flintlock_Reload_End_Cue
root=WavePlayer_0
nodes [{id=WavePlayer_0, properties {soundWaveAssetPtr=/Game/Sounds/Fort_Weapons/FlintLock/Flintlock_Revolver_Cock}}]
properties {
 attenuationSettings=&Fort_Weapon_Reload_Att
 bOverrideConcurrency=true
 maxDistance=2250
 pitchMultiplier=0.9
 soundClassObject=&Fort_WeaponFoley
 volumeMultiplier=0.351
 concurrencyOverrides {resolutionRule=PreventNew}
}
