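# Sound cue for a flintlock pistol reload insert step, classed as weapon foley with reload attenuation.
# A mixer layers a random dual-pistol reload-start sound with a quieter, briefly delayed random tube-insert sound that passes through its own modulator.
# An outer modulator varies volume slightly while holding pitch fixed.
# A concurrency override prevents new instances once the limit is reached.
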SoundCue /Game/Sounds/Fort_Weapons/pistol/Flintlock/Pistol_Flintlock_Reload_Insert2_Cue
root=Modulator_3
nodes [
 {id=Delay_0, children [Random_0], properties {delayMax=0.05, delayMin=0.05}}
 {id=Mixer_1, children [Random_1,Modulator_0], properties {inputVolume [1.2,0.34]}}
 {id=Modulator_0, children [Delay_0], properties {pitchMin=1}}
 {id=Modulator_3, children [Mixer_1], properties {pitchMax=1, pitchMin=1, volumeMax=1.3, volumeMin=1.2}}
 {id=Random_0, children [WavePlayer_1,WavePlayer_2], properties {weights [1,1]}}
 {id=Random_1, children [WavePlayer_0,WavePlayer_3,WavePlayer_4], properties {weights [1,1,1]}}
 {id=WavePlayer_0, properties {soundWaveAssetPtr=/Game/Sounds/Fort_Weapons/pistol/Dual/DualPistols_Reload_Start_01}}
 {id=WavePlayer_1, properties {soundWaveAssetPtr=&/Pistol_Flintlock_InsertTube_02}}
 {id=WavePlayer_2, properties {soundWaveAssetPtr=&/Pistol_Flintlock_InsertTube_03}}
 {id=WavePlayer_3, properties {soundWaveAssetPtr=/Game/Sounds/Fort_Weapons/pistol/Dual/DualPistols_Reload_Start_02}}
 {id=WavePlayer_4, properties {soundWaveAssetPtr=/Game/Sounds/Fort_Weapons/pistol/Dual/DualPistols_Reload_Start_03}}
]
properties {
 attenuationSettings=&Fort_Weapon_Reload_Att
 bOverrideConcurrency=true
 maxDistance=2250
 soundClassObject=&Fort_WeaponFoley
 volumeMultiplier=0.501
 concurrencyOverrides {resolutionRule=PreventNew}
}
