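# Sound cue for the insert stage of the flintlock pistol reload, classed as weapon foley.
# A single wave player plays a flintlock revolver reload-insert sample through a modulator node at the root.
# The cue lowers pitch and volume, uses the weapon reload attenuation, and overrides concurrency to prevent new instances instead of stopping existing ones.
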SoundCue /Game/Sounds/Fort_Weapons/pistol/Flintlock/Pistol_Flintlock_Reload_Insert_Cue
root=Modulator_0
nodes [
 {id=Modulator_0, children [WavePlayer_0]}
 {id=WavePlayer_0, properties {soundWaveAssetPtr=/Game/Sounds/Fort_Weapons/FlintLock/Flintlock_Revolver_ReloadInsert_01}}
]
properties {
 attenuationSettings=&Fort_Weapon_Reload_Att
 bOverrideConcurrency=true
 maxDistance=2250
 pitchMultiplier=0.85
 soundClassObject=&Fort_WeaponFoley
 volumeMultiplier=0.351
 concurrencyOverrides {resolutionRule=PreventNew}
}
