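# Sound cue for the start of the flintlock pistol's reload, classed as weapon foley.
# A mixer layers a player turn foley wave, passed through a modulator and weighted louder, with one of three equally weighted random reload-start waves borrowed from the dual pistols.
# It plays quietly through the shared weapon reload attenuation.
# A concurrency override prevents new instances once the limit is reached.
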SoundCue /Game/Sounds/Fort_Weapons/pistol/Flintlock/Pistol_Flintlock_Reload_Start_Cue
root=Mixer_1
nodes [
 {id=Mixer_1, children [Modulator_2,Random_1], properties {inputVolume [1.5,0.6]}}
 {id=Modulator_2, children [WavePlayer_3]}
 {id=Random_1, children [WavePlayer_4,WavePlayer_5,WavePlayer_6], properties {weights [1,1,1]}}
 {id=WavePlayer_3, properties {soundWaveAssetPtr=/Game/Sounds/Fort_Foley_Sounds/Moves/PlayerFoley_Turn_03}}
 {id=WavePlayer_4, properties {soundWaveAssetPtr=/Game/Sounds/Fort_Weapons/pistol/Dual/DualPistols_Reload_Start_01}}
 {id=WavePlayer_5, properties {soundWaveAssetPtr=/Game/Sounds/Fort_Weapons/pistol/Dual/DualPistols_Reload_Start_02}}
 {id=WavePlayer_6, properties {soundWaveAssetPtr=/Game/Sounds/Fort_Weapons/pistol/Dual/DualPistols_Reload_Start_03}}
]
properties {
 attenuationSettings=&Fort_Weapon_Reload_Att
 bOverrideConcurrency=true
 maxDistance=2250
 soundClassObject=&Fort_WeaponFoley
 volumeMultiplier=0.351
 concurrencyOverrides {resolutionRule=PreventNew}
}
